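# Sound cue for the looping ambient sound of the Rift Beacon beam in the Figment S10 event content.
# A Switch on the integer parameter Stage plays one looping rift panel wave for stages 3 and 4, while the default and other inputs stay empty.
# Inline attenuation overrides the referenced Rift Beacon attenuation asset with a very long falloff, and the cue uses the medium-priority events sound class.
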SoundCue /Figment_S10/Athena/Sounds/Events/RiftBeacon/RiftBeacon_Beam_Amb_Loop_Cue
root=Switch_0
nodes [
 {id=Switch_0, children [null,null,null,null,WavePlayer_0,WavePlayer_0,null], properties {intParameterName=Stage}}
 {id=WavePlayer_0, properties {bLooping=true, soundWaveAssetPtr=/Figment_S10/Athena/Sounds/Events/RiftBeacon/AssetDrop_01/Rift_PanelLoop_01}}
]
properties {
 attenuationSettings=/Game/Athena/Sounds/Events/RiftBeacon/RiftBeacon_Normal_Att
 bOverrideAttenuation=true
 maxDistance=1000400
 soundClassObject=/Figment_S10/Athena/Sounds/Events/SoundClasses/Fort_Events_Priority_Medium
 attenuationOverrides {falloffDistance=1000000}
}
