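# Sound Cue for a one-shot at the inside edge of the Rift Beacon bubble, judging by its name, in the Figment Season 10 events audio.
# A Random node picks one of three equally weighted Rift gravity-flare waves, played at reduced volume in a medium-priority events sound class.
# Inline attenuation overrides supersede the referenced outside-ambience attenuation asset, disabling volume falloff while keeping a distance-based low-pass filter and a wide stereo spread.
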SoundCue /Figment_S10/Athena/Sounds/Events/RiftBeacon/RiftBeacon_Bubble_Inside_Edge_Oneshot_Cue
root=Random_0
nodes [
 {id=Random_0, children [WavePlayer_0,WavePlayer_2,WavePlayer_3], properties {weights [1,1,1]}}
 {id=WavePlayer_0, properties {soundWaveAssetPtr=/Figment_S10/Athena/Sounds/Events/RiftBeacon/AssetDrop_01/Rift_GravityFlare_03}}
 {id=WavePlayer_2, properties {soundWaveAssetPtr=/Figment_S10/Athena/Sounds/Events/RiftBeacon/AssetDrop_01/Rift_GravityFlare_01}}
 {id=WavePlayer_3, properties {soundWaveAssetPtr=/Figment_S10/Athena/Sounds/Events/RiftBeacon/AssetDrop_01/Rift_GravityFlare_02}}
]
properties {
 attenuationSettings=/Game/Athena/Sounds/Events/RiftBeacon/RiftBeacon_Bubble_Amb_Outside_Att
 bOverrideAttenuation=true
 maxDistance=2097152
 soundClassObject=/Figment_S10/Athena/Sounds/Events/SoundClasses/Fort_Events_Priority_Medium
 volumeMultiplier=0.2
 attenuationOverrides {bAttenuate=false, bAttenuateWithLPF=true, lPFFrequencyAtMax=200, lPFRadiusMax=4000, lPFRadiusMin=2000, stereoSpread=4096}
}
