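# Sound cue for the looping ambient sound of the Rift Beacon cable in a Figment Season 10 event.
# A single looping wave player feeds the output directly with the Rift_CableLoop_01 wave, with no mixing or randomization.
# It uses the shared close-range Rift Beacon attenuation from the base game and a medium-priority event sound class.
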SoundCue /Figment_S10/Athena/Sounds/Events/RiftBeacon/RiftBeacon_Cable_Amb_Loop_Cue
root=WavePlayer_0
nodes [{id=WavePlayer_0, properties {bLooping=true, soundWaveAssetPtr=/Figment_S10/Athena/Sounds/Events/RiftBeacon/AssetDrop_01/Rift_CableLoop_01}}]
properties {
 attenuationSettings=/Game/Athena/Sounds/Events/RiftBeacon/RiftBeacon_Close_Att
 maxDistance=1536
 soundClassObject=/Figment_S10/Athena/Sounds/Events/SoundClasses/Fort_Events_Priority_Medium
}
