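# Rift beacon crack ambience cue that plays a looping rift intensity sound chosen by the beacon's stage.
# A switch on the Stage integer parameter leaves its first four inputs empty, so those play nothing.
# The later inputs go to the standard intensity loop, then the high intensity loop, then back to the standard one through a reused mixer.
# It uses dedicated crack ambience attenuation and the medium-priority events sound class.
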SoundCue /Figment_S10/Athena/Sounds/Events/RiftBeacon/RiftBeacon_Crack_Amb_Loop_Cue
root=Switch_0
nodes [
 {id=Mixer_0, children [WavePlayer_17, null], properties {inputVolume [1,1]}}
 {id=Mixer_1, children [null, WavePlayer_18], properties {inputVolume [1,1]}}
 {id=Switch_0, children [null,null,null,null,Mixer_0,Mixer_1,Mixer_0], properties {intParameterName=Stage}}
 {id=WavePlayer_17, properties {bLooping=true, soundWaveAssetPtr=/Figment_S10/Athena/Sounds/Events/RiftBeacon/AssetDrop_01/Rift_Intensity_01}}
 {
  id=WavePlayer_18
  properties {bLooping=true, soundWaveAssetPtr=/Figment_S10/Athena/Sounds/Events/RiftBeacon/AssetDrop_01/Rift_IntensityHigh_01}
 }
]
properties {
 attenuationSettings=&/RiftBeacon_Crack_Amb_Att
 maxDistance=2097152
 soundClassObject=/Figment_S10/Athena/Sounds/Events/SoundClasses/Fort_Events_Priority_Medium
}
