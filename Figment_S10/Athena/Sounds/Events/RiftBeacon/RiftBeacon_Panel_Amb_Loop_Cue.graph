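# Sound cue for the looping ambience of a rift beacon panel, gated by the beacon's Stage parameter.
# A switch on Stage leaves the first three inputs empty and routes the last three to one shared mixer, so only later stages are audible.
# The mixer layers a boosted panel loop with normal attenuation over a quieter emitter mechanism loop with close attenuation.
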SoundCue /Figment_S10/Athena/Sounds/Events/RiftBeacon/RiftBeacon_Panel_Amb_Loop_Cue
root=Switch_0
nodes [
 {id=Attenuation_0, children [WavePlayer_0], properties {attenuationSettings=/Game/Athena/Sounds/Events/RiftBeacon/RiftBeacon_Normal_Att}}
 {id=Attenuation_1, children [WavePlayer_2], properties {attenuationSettings=/Game/Athena/Sounds/Events/RiftBeacon/RiftBeacon_Close_Att}}
 {id=Mixer_0, children [Attenuation_0, Attenuation_1], properties {inputVolume [2,0.25]}}
 {id=Switch_0, children [null,null,null,Mixer_0,Mixer_0,Mixer_0], properties {intParameterName=Stage}}
 {id=WavePlayer_0, properties {bLooping=true, soundWaveAssetPtr=/Figment_S10/Athena/Sounds/Events/RiftBeacon/AssetDrop_01/Rift_PanelLoop_01}}
 {id=WavePlayer_2, properties {bLooping=true, soundWaveAssetPtr=/Game/Athena/Sounds/Events/RiftBeacon/AssetDrop_01/Rift_EmitterMech_01}}
]
properties {maxDistance=2000, soundClassObject=/Figment_S10/Athena/Sounds/Events/SoundClasses/Fort_Events_Priority_Medium}
